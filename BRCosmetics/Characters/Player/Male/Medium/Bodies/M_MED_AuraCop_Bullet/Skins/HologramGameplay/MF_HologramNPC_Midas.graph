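# Material function for the Midas hologram NPC skin that turns incoming material attributes into a tinted, scanlined, see-through hologram look.
# Fresnel blends biased base color with a fake transparency from skylight cubemap sampling along a camera-derived vector, boosted at night.
# Tint and sun passthrough feed emissive, brightened for ray tracing, and a scaled copy replaces base color.
# Low-quality paths use time-of-day colors for the cubemap and drop scanlines and sunlight.
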MaterialFunction /BRCosmetics/Characters/Player/Male/Medium/Bodies/M_MED_AuraCop_Bullet/Skins/HologramGameplay/MF_HologramNPC_Midas
graph {nodes [
  n1 = GetMaterialAttributes(None=FunctionInput() {inputType=FunctionInput_MaterialAttributes}) {attributeGetTypes [#1, #2]}
  n2 = VectorParameter() {group=AuraCop, parameterName=ColorBias, defaultValue=<rgba 0.02 0.02 0.02 1>}
  n3 = Add(A=@n1.BaseColor, B=@n2.RGB)
  n4 = NamedRerouteUsage()
  n5 = Fresnel(ExponentIn=ScalarParameter() {defaultValue=3, parameterName=SeethroughSidesFres}) {baseReflectFraction=0}
  n6 = Saturate(None=LinearInterpolate(B=ScalarParameter() {defaultValue=3, parameterName=See Through Side Max}, Alpha=Saturate(None=@n5)) {constB=0})
  n7 = Clamp(None=Multiply(A=LinearInterpolate(A=@n3, B=@n4, Alpha=@n6), B=NamedRerouteUsage()))
  n8 = VectorParameter() {parameterName=ColorTint, defaultValue=<rgba 0.2919 0.5393 1 1>}
  n9 = Multiply(A=@n8.RGB, B=@n8.RGB)
  n10 = ComponentMask(None=Saturate(None=CollectionParameter() {collection=&FortniteMaterialParameters, parameterName=SunLightColor})) {b=true, g=true, r=true}
  n11 = NamedRerouteUsage()
  n12 = Multiply(
   A=Multiply(A=ScalarParameter() {defaultValue=0.02, parameterName=Sunlight Passthrough}, B=Multiply(A=@n10, B=@n11))
   B=Clamp(None=MaterialFunctionCall() {materialFunction=/Game/Packages/Fortress_Sky/Fortnite_SunScreenSpaceFalloff}.Dot*-1) {minDefault=0.5}
  )
  n13 = Constant()
  n14 = Add(A=@n1.EmissiveColor, B=Add(A=Multiply(A=@n7, B=@n9), B=QualitySwitch(Default=@n12, Low=@n13)))
  n15 = Multiply(A=@n14, B=ScalarParameter() {defaultValue=8, parameterName=LumenBrightness})
  n16 = ScalarParameter() {defaultValue=0.4, parameterName=ColorNess}
  FunctionOutput(None=SetMaterialAttributes!(
    MaterialAttributes=@n1.MaterialAttributes
    EmissiveColor=RayTracingQualitySwitch(Normal=@n14, RayTraced=@n15)
    BaseColor=Multiply(A=@n14, B=@n16)
   ) {attributeSetTypes [#2, #1]})
  Multiply(A=@n1.BaseColor, B=Add(B=@n2.RGB) {constA=1})
  n17 = Constant() {r=1}
  n18 = MaterialFunctionCall(Value_Morning=@n17, Value_Day=@n17, Value_Evening=@n17, Value_Night=ScalarParameter() {defaultValue=5, parameterName=Night Boost}) {materialFunction=&Scalar_Time_of_Day}
  Multiply(A=@n18.Result) {constB=0.5}
  n19 = Constant() {r=200}
  NamedRerouteDeclaration(None=MaterialFunctionCall!(
    Default=MaterialFunctionCall!(Scanline Amount=@n19) {materialFunction=/Game/Environments/Asteria/Sets/NeonCity/Techart/Holograms/NewHologramMaterial/MFScanlines}.Result
    Low/Mobile=Constant() {r=1}
   ) {materialFunction=/Game/Effects/Fort_Effects/Materials/_Functions/MF_LowMobile_PathingSwitch}.Result) {name=Scanlines, nodeColor=<rgba 1 0 0.1641 1>}
  n20 = MaterialFunctionCall(Float3=OneMinus(None=CameraVectorWS())) {materialFunction=&BreakOutFloat3Components}
  n21 = MaterialFunctionCall(X=@n20.R, Y=@n20.R, Z=Add(A=@n20.B, B=ScalarParameter() {defaultValue=-1, parameterName=Offset SeeThrough})) {materialFunction=&MakeFloat3}
  n22 = ScalarParameter() {defaultValue=0.1, parameterName=HologramRoughness}
  n23 = Multiply(
   A=Clamp(None=SkyLightEnvMapSample(Direction=@n21.Result, Roughness=@n22)) {maxDefault=4, minDefault=0.5}
   B=ScalarParameter() {defaultValue=1, parameterName=CubemapBleed}
  )
  n24 = VectorParameter() {parameterName=ColorMorning, defaultValue=<rgba 0.2031 0.1723 0.1315 1>}
  n25 = VectorParameter() {parameterName=ColorNight, defaultValue=<rgba 0.0373 0.0718 0.1302 1>}
  NamedRerouteDeclaration(None=QualitySwitch(
    Default=Multiply(A=@n23, B=@n18.Result)
    Low=MaterialFunctionCall(Color_Morning=@n24.RGB, Color_Day=@n24.RGB, Color_Evening=@n24.RGB, Color_Night=@n25.RGB) {materialFunction=&Color_Time_of_Day}.Result
   )) {name=Cubemap, nodeColor=<rgba 0.8984 0 1 1>}
  Saturate()
 ]}
