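# Parent character material for the AuraCop Bullet hologram NPC skin, built for skeletal meshes with morph targets and mesh deformers.
# It wraps a hologram effect, MF_HologramNPC_Midas, between the shared 2023 CharacterParent PreFX and PostFX functions.
# Named reroutes carry material attributes between stages, and PreFX also feeds mask, pre-skinned position, distance blend and Fresnel inputs on PostFX.
# It uses material attributes, with the shading model chosen by expression.
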Material /BRCosmetics/Characters/Player/Male/Medium/Bodies/M_MED_AuraCop_Bullet/Skins/HologramGameplay/M_AuraCop_Parent_HologramNPC
graph {
 properties {
  bUseMaterialAttributes=true
  bUsedWithMeshDeformer=true
  bUsedWithMorphTargets=true
  bUsedWithSkeletalMesh=true
  pixelDepthOffsetMode=PDOM_Legacy
  shadingModel=MSM_FromMaterialExpression
  usedShadingModels=Default Lit | Subsurface | Clear Coat | Subsurface Profile | Cloth
 }
 nodes [
  n1 = MaterialFunctionCall() {materialFunction=&MF_CharacterParent_2023_PreFX}
  MaterialFunctionCall(Material=NamedRerouteUsage(), M=@n1.M, Pre Skinned Local Position=@n1.M, Distance Blend=@n1.M, Fresnel=@n1.M) {materialFunction=&MF_CharacterParent_2023_PostFX}
  NamedRerouteDeclaration(None=@n1.Result) {name=InMaterialAttributes, nodeColor=<rgba 0 0.2266 1 1>}
  NamedRerouteDeclaration(None=MaterialFunctionCall!(In=NamedRerouteUsage()) {materialFunction=&/MF_HologramNPC_Midas}.Result) {name=OutMaterialAttributes, nodeColor=<rgba 0.5469 1 0 1>}
  NamedRerouteDeclaration(None=@n1.M) {name=InMask, nodeColor=<rgba 1 0 0.9375 1>}
 ]
}
